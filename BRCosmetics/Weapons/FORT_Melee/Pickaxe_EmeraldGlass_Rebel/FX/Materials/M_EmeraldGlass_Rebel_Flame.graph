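# Unlit additive material for flame effects on the EmeraldGlass Rebel pickaxe, used on Niagara sprite and mesh particles.
# A grayscale flame burst texture, with UV scale, offset and optional panning over particle relative time, is tinted by particle color and passed through an emissive adjustment function.
# Opacity combines texture and particle alpha, with switchable depth fade that mobile skips, then a switchable double-sided Fresnel side fade.
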Material /BRCosmetics/Weapons/FORT_Melee/Pickaxe_EmeraldGlass_Rebel/FX/Materials/M_EmeraldGlass_Rebel_Flame
graph {
 properties {
  bUsedWithNiagaraMeshParticles=true
  bUsedWithNiagaraSprites=true
  blendMode=BLEND_Additive
  pixelDepthOffsetMode=PDOM_Legacy
  shadingModel=MSM_Unlit
  usedShadingModels=Unlit
 }
 nodes [
  n1 = TextureCoordinate()
  n2 = ComponentMask(None=@n1) {r=true}
  n3 = ScalarParameter() {defaultValue=1, parameterName=U_Scale}
  n4 = ComponentMask(None=@n1) {g=true}
  n5 = ScalarParameter() {defaultValue=1, parameterName=V_Scale}
  n6 = AppendVector(
   A=Add(A=Multiply(A=@n2, B=@n3), B=ScalarParameter() {parameterName=U_Offset})
   B=Add(A=Multiply(A=@n4, B=@n5), B=ScalarParameter() {parameterName=V_Offset})
  )
  n7 = Panner(
   Coordinate=@n6
   Time=ParticleRelativeTime()
   Speed=AppendVector(A=ScalarParameter() {parameterName=U_Speed}, B=ScalarParameter() {parameterName=V_Speed})
  )
  n8 = TextureSampleParameter2D(UVs=StaticSwitchParameter(True=StaticSwitchParameter(True=@n7, False=@n6) {parameterName=Use_Panner}, False=TextureCoordinate()) {defaultValue=true, parameterName=Use_UV_Controls}) {
   group=Tex1
   parameterName=Tex1
   samplerType=SAMPLERTYPE_LinearGrayscale
   texture=/BRCosmetics/Weapons/FORT_Melee/Pickaxe_EmeraldGlass_Rebel/FX/Textures/T_Burst_Flame
  }
  n9 = StaticSwitchParameter(True=@n8.R, False=@n8.R) {defaultValue=true, group=Tex1, parameterName=Use_only_Red_Channel}
  n10 = ParticleColor()
  MaterialFunctionCall(Emissive=Multiply(A=@n9, B=@n10.RGB)) {materialFunction=&MF_FX_EmissiveAdjustments}
  n11 = Multiply(A=StaticSwitchParameter(True=@n9, False=@n8.A) {defaultValue=true, group=Tex1, parameterName=Use_Color_as_Alpha}, B=@n10.A)
  n12 = StaticSwitchParameter(
   True=DepthFade(Opacity=@n11, FadeDistance=ScalarParameter() {defaultValue=5, parameterName=Depth_Fade_Distance}) {fadeDistanceDefault=5}
   False=@n11
  ) {defaultValue=true, parameterName=Use_Depth_Fade}
  n13 = ScalarParameter() {defaultValue=1, group=Alpha, parameterName="CameraNormal_Fresnel_Brightness (S)"}
  n14 = ScalarParameter() {defaultValue=3, parameterName=Fade edge}
  n15 = VertexNormalWS()
  Multiply(
   A=ShadingPathSwitch(Default=@n12, Deferred=@n12, Mobile=@n11) {desc=Migrated from FeatureLevelSwitch}
   B=StaticSwitchParameter(
    True=MaterialFunctionCall(Brightness=@n13, Exponent=@n14, Vertex / Pixel Normal=@n15) {materialFunction=&MF_Fresnel_FX}.Fresnel Double Sided
    False=Constant() {r=1}
   ) {defaultValue=true, parameterName=Side_Fade}
  )
  ScalarParameter() {defaultValue=3, group=Alpha, parameterName="CameraNormal_Fresnel_Exponent (S)"}
 ]
}
